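# Surface material for the C4 explosive, usable on skeletal meshes, with a default-lit textured body and glowing details.
# A sine of time pulses a green glow through the mask texture's red channel, added to a yellow arrow decal placed by scaled, offset UVs.
# The sum feeds MF_FX_EmissiveAdjustments, while the base texture goes through ConvertFromDiffSpec.
# The Arrow_Direction parameter blends the same channel with itself, so it has no effect.
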Material /Game/Weapons/FORT_Explosives/Materials/M_C4
graph {
 properties {bUsedWithMeshDeformer=true, bUsedWithSkeletalMesh=true, pixelDepthOffsetMode=PDOM_Legacy, usedShadingModels=Default Lit}
 nodes [
  n1 = Sine(None=Time())
  n2 = Constant() {r=1}
  n3 = Multiply(A=Constant3Vector() {constant=<rgba 0 10 0 0>}, B=Desaturation(None=Add(A=@n1, B=@n2)))
  n4 = Desaturation(None=TextureSample() {texture=/Game/Weapons/FORT_Explosives/Textures/T_C4_Masks}.R)
  n5 = TextureCoordinate()
  n6 = Constant() {r=3.4}
  n7 = Constant() {r=-2.39}
  n8 = Constant() {r=-2.4}
  n9 = TextureSample(UVs=Add(A=Multiply(A=@n5, B=@n6), B=AppendVector(A=@n7, B=@n8))) {texture=/Game/Weapons/FORT_Explosives/Textures/T_C4_Arrows}
  n10 = LinearInterpolate(A=@n9.R, B=@n9.R, Alpha=ScalarParameter() {defaultValue=1, parameterName=Arrow_Direction})
  n11 = Constant3Vector() {constant=<rgba 2 2 0 0>}
  MaterialFunctionCall(Emissive=Add(A=Multiply(A=@n3, B=@n4), B=Multiply(A=@n10, B=@n11))) {materialFunction=&MF_FX_EmissiveAdjustments}
  MaterialFunctionCall(DiffuseColor=TextureSample() {texture=/Game/Weapons/FORT_Explosives/Textures/T_C4}.RGB) {materialFunction=&ConvertFromDiffSpec}
 ]
}
